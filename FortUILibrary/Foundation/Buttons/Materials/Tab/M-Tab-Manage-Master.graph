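# Translucent UI master material, named for Manage tab buttons, that builds a parameterized two-color gradient and a slanted shape mask.
# The gradient blends RGBA0 and RGBA1 through SmoothStep over rotated, normalized UI UVs.
# The mask skews offset UVs by Angle, applies Scale, then uses absolute value and SmoothStep to cut a centered band, multiplied by Alpha.
# A Max node with identical inputs passes the scaled value through unchanged.
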Material /FortUILibrary/Foundation/Buttons/Materials/Tab/M-Tab-Manage-Master
graph {
 properties {
  bUsedWithStaticMesh=false
  blendMode=BLEND_Translucent
  materialDomain=MD_UI
  pixelDepthOffsetMode=PDOM_Legacy
  shadingModel=MSM_Unlit
  usedShadingModels=Unlit
 }
 nodes [
  n1 = Rotator(
   Coordinate=MaterialFunctionCall() {materialFunction=&GetUserInterfaceUV}.Normalized UV
   Time=ScalarParameter() {defaultValue=1, group=Linear Gradient Settings, parameterName=Rotation}
  )
  LinearInterpolate(
   A=VectorParameter() {parameterName=RGBA0, defaultValue=<rgba 0.61 0.1365 0.5303 1>}.RGB
   B=VectorParameter() {parameterName=RGBA1, defaultValue=<rgba 0.4263 0.57 0.4461 1>}.RGB
   Alpha=MaterialFunctionCall(
    Alpha=ComponentMask(None=@n1) {g=true}
    Min=ScalarParameter() {group=Linear Gradient Settings, parameterName=LG SS Min}
    Max=ScalarParameter() {defaultValue=1, group=Linear Gradient Settings, parameterName=LG SS Max}
   ) {materialFunction=&SmoothStep}.Result
  )
  n2 = MaterialFunctionCall(Float2=Subtract(A=TextureCoordinate(), B=ScalarParameter() {defaultValue=0.5, parameterName=Offset})) {materialFunction=&BreakOutFloat2Components}
  n3 = ScalarParameter() {defaultValue=24, parameterName=Angle}
  n4 = Multiply(
   A=Add(A=Multiply(A=@n2.R, B=@n3), B=@n2.G)
   B=ComponentMask(None=VectorParameter() {parameterName=Scale, defaultValue=<rgba 0.1567 0 0 1>}.RGB) {g=true, r=true}
  )
  n5 = Max(A=@n4, B=@n4) {constB=0}
  Multiply(
   A=MaterialFunctionCall(
    Alpha=Abs(None=@n5)
    Min=ScalarParameter() {defaultValue=0.6, parameterName=SS Min}
    Max=ScalarParameter() {defaultValue=0.5835, parameterName=SS Max}
   ) {materialFunction=&SmoothStep}.Result
   B=ScalarParameter() {defaultValue=1, parameterName=Alpha}
  )
 ]
}
